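; Mesh capture rig for the 3DHB texture baker, rendering meshes through diffuse, depth, unlit and normal scene capture components.
; Construction adds children of ShowList to the normal and depth captures' show-only lists, and children of OnlyNormalShowList to the normal capture alone.
; Initial installs a supplied dynamic material as a post-process blendable on the depth capture.
; ToggleCaptureFreeze flips every-frame capturing on all four captures, based on the normal capture's state.
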(event EventBeginPlay)

(event Custom|Initial
  (params (PostDepthMID :type Material%20Instance%20Dynamic%20Object%20Reference))
  (Utilities|IsValid PostDepthMID
    (:Is%20Valid
      (Variables|/Script/Engine.SceneCaptureComponent2D|SetPostProcessSettings (Utilities|Struct|MakePostProcessSettings (Utilities|Struct|MakeWeightedBlendables (Utilities|Array|MakeArray (Utilities|Struct|MakeWeightedBlendable 1.0 PostDepthMID)))) (Variables|Default|GetDepthCapture)))))

(event Collision|EventActorBeginOverlap
  (params (OtherActor :type Actor%20Object%20Reference)))

(event EventTick
  (params (DeltaSeconds :type Float%20%28single-precision%29)))

(event Custom|ToggleCaptureFreeze
  (bind _normalcapture (Variables|Default|GetNormalCapture))
  (bind _bcaptureeveryframe (Variables|/Script/Engine.SceneCaptureComponent|GetbCaptureEveryFrame _normalcapture))
  (bind _returnvalue (not _bcaptureeveryframe))
  (Variables|/Script/Engine.SceneCaptureComponent|SetbCaptureEveryFrame _returnvalue (Variables|Default|GetDiffuseCapture))
  (Variables|/Script/Engine.SceneCaptureComponent|SetbCaptureEveryFrame _returnvalue (Variables|Default|GetDepthCapture))
  (Variables|/Script/Engine.SceneCaptureComponent|SetbCaptureEveryFrame _returnvalue (Variables|Default|GetUnlitCapture))
  (Variables|/Script/Engine.SceneCaptureComponent|SetbCaptureEveryFrame _returnvalue _normalcapture))

(fn ConstructionScript
  (Utilities|FlowControl|Sequence
    (:then_0
      (for _array_element (Components|GetChildrenComponents (Variables|Default|GetOnlyNormalShowList) true)
        (bind _asprimitive_component (Utilities|Casting|CastToPrimitiveComponent _array_element))
        (Rendering|SceneCapture|ShowOnlyComponent (Variables|Default|GetNormalCapture) _asprimitive_component)))
    (:then_1
      (for _array_element_2 (Components|GetChildrenComponents (Variables|Default|GetShowList) true)
        (bind _asprimitive_component_2 (Utilities|Casting|CastToPrimitiveComponent _array_element_2))
        (Rendering|SceneCapture|ShowOnlyComponent (Variables|Default|GetNormalCapture) _asprimitive_component_2)))
    (:then_2
      (for _array_element_3 (Components|GetChildrenComponents (Variables|Default|GetShowList) true)
        (bind _asprimitive_component_3 (Utilities|Casting|CastToPrimitiveComponent _array_element_3))
        (Rendering|SceneCapture|ShowOnlyComponent (Variables|Default|GetDepthCapture) _asprimitive_component_3)))))
