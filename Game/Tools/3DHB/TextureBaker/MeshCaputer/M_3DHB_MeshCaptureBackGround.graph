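# Background material for the mesh capture part of the 3DHB texture baker tool, producing an animated, hue-cycling grid.
# World-space Y and negated X are scaled into UVs for the grid texture, so the pattern does not depend on mesh UVs.
# World X plus Time drives HueShift on the sampled grid, giving color bands that scroll along the X axis.
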Material /Game/Tools/3DHB/TextureBaker/MeshCaputer/M_3DHB_MeshCaptureBackGround
graph {
 properties {pixelDepthOffsetMode=PDOM_Legacy, usedShadingModels=Default Lit}
 nodes [
  Constant()
  n1 = MaterialFunctionCall(Float3=WorldPosition().XYZ) {materialFunction=&BreakOutFloat3Components}
  n2 = Divide(A=@n1.R) {constB=250}
  n3 = Time()
  n4 = AppendVector(A=@n1.G, B=Multiply(A=@n1.R) {constB=-1})
  MaterialFunctionCall(
   Hue Shift Percentage=Multiply(A=Add(A=@n2, B=@n3)) {constB=0.2}
   Texture=TextureSample(UVs=Divide(A=@n4) {constB=250}) {texture=/Game/Tools/3DHB/DefaultAssets/T_3DHB_Grid}.RGB
  ) {materialFunction=&HueShift}
 ]
}
